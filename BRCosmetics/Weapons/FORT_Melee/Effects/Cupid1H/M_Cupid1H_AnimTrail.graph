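# Unlit, two-sided translucent beam-trail material for the swing trail of the Cupid1H melee weapon.
# A floored UV gradient cuts a hard, tapering band set by Blade Length, multiplied by a fade and an arc mask panned via Wind Move.
# That band takes particle color with HDR remapping, plus a wind layer from the same mask, Wind Color and dynamic Wind Intensity.
# Particle alpha scales both layers for opacity.
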Material /BRCosmetics/Weapons/FORT_Melee/Effects/Cupid1H/M_Cupid1H_AnimTrail
graph {
 properties {
  bUsedWithBeamTrails=true
  blendMode=BLEND_Translucent
  pixelDepthOffsetMode=PDOM_Legacy
  shadingModel=MSM_Unlit
  twoSided=true
  usedShadingModels=Unlit
 }
 nodes [
  n1 = TextureCoordinate()
  n2 = ComponentMask(None=TextureCoordinate()) {r=true}
  n3 = Add(
   A=OneMinus(None=ComponentMask(None=@n1) {g=true})
   B=LinearInterpolate(B=ScalarParameter() {defaultValue=0.05, parameterName=Blade Length}, Alpha=OneMinus(None=@n2))
  )
  n4 = TextureSample(UVs=Panner(Coordinate=TextureCoordinate(), Time=ScalarParameter() {parameterName=Wind Move}) {speedX=-1}) {texture=/Game/Effects/Fort_Effects/Textures/Melee/T_Trail_ArcMask}
  n5 = TextureCoordinate()
  n6 = Multiply(A=Multiply(A=Floor(None=@n3), B=@n4.R), B=OneMinus(None=ComponentMask(None=@n5) {r=true}))
  n7 = Multiply(A=ParticleColor().RGB, B=@n6)
  n8 = Clamp(None=@n7)
  n9 = Multiply(
   A=@n4.R
   B=Multiply(
    A=VectorParameter() {parameterName=Wind Color, defaultValue=<rgba 1 1 1 0>}.RGB
    B=DynamicParameter() {defaultValue=<rgba 0.25 1 1 1>, paramNames [Wind Intensity, Param2, Param3, Param4]}.Wind Intensity
   )
  )
  MaterialFunctionCall(Emissive=Add(
    A=MaterialFunctionCall(LDR Input=@n8, HDR Tint=@n7) {materialFunction=&DeriveHDRfromLDR}.RemappedRange
    B=@n9
   )) {materialFunction=&MF_FX_EmissiveAdjustments}
  Reroute(None=Multiply(A=ParticleColor().A, B=Add(A=@n6, B=@n9)))
 ]
}
